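# Niagara system for the Swoosh emote that draws a soft, short-lived, translucent white ribbon trail from an animated skeletal mesh.
# Its single SoftTrail emitter uses AnimTrail functions to spawn, initialize and update particles against two bone or socket indices on the user-supplied mesh.
# A user Transparency parameter scales alpha every update, while two emitter-age fade curves are present but disabled.
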NiagaraSystem /BRCosmetics/Effects/Fort_Effects/Emotes/Swoosh/NS_Emote_Swoosh_Trail
dependencies {usedDataInterfaces [&NiagaraDataInterfaceCurve], usedDynamicInputs [], usedModules [&SystemState]}
emitters {SoftTrail {
  emitterUpdateScript [
   &EmitterState()
   &FN_SpawnAnimTrail()
  ]
  particleSpawnScript [
   &FN_InitializeAnimTrailBoneIndecies(First Bone or Socket Index=1, Second Bone or Socket Index=0)
   &FN_InitializeAnimTrail(Color=<rgba 1 1 1 0.25>, Lifetime=0.25) {module=FN_InitializeAnimTrail001}
  ]
  particleUpdateScript [
   &FN_UpdateAnimTrail(SkeletalMesh=$User.SkeletalMesh) {module=FN_UpdateAnimTrail001}
   &ScaleColor(
    Scale Alpha=&FloatFromCurve(
     FloatCurve=DataInterface() {propertyValues {Curve {keys [
         {tangentMode=RCTM_User, time=0, value=1}
         {arriveTangent=-0.0405, interpMode=RCIM_Cubic, leaveTangent=-0.0405, tangentMode=RCTM_User, time=1.1, value=1}
         {interpMode=RCIM_Cubic, leaveTangent=-8.0487, time=1.5, value=0}
        ]}}}
     CurveIndex=$Emitter.Age
    )
    ScaleA=true
    ScaleRGB=false
   ) {enabled=false, module=ScaleColor001}
   &ScaleColor(
    Scale Alpha=&FloatFromCurve(
     FloatCurve=DataInterface() {propertyValues {Curve {keys [
         {interpMode=RCIM_Cubic, tangentMode=RCTM_User, time=0, value=1}
         {interpMode=RCIM_Cubic, leaveTangent=-5.2892, tangentMode=RCTM_Break, time=0.3, value=1}
         {interpMode=RCIM_Cubic, tangentMode=RCTM_User, time=1, value=0}
        ]}}}
     CurveIndex=$Emitter.Age
    )
    ScaleA=true
    ScaleRGB=false
   ) {enabled=false}
   &SolveForcesAndVelocity(Clamp Velocity=false, Limit Acceleration=false)
   &ScaleColor(Scale Alpha=$User.Transparency, ScaleA=true, ScaleRGB=false) {module=ScaleColor002}
  ]
  rendererClasses [&NiagaraRibbonRendererProperties]
  renderers [{rendererClass=&NiagaraRibbonRendererProperties}]
 }}
summary {
 systemName=NS_Emote_Swoosh_Trail
 emitters [{emitterName=SoftTrail, rendererClasses [&NiagaraRibbonRendererProperties]}]
 userVariables [
  User.RibbonIDOffset {type=NiagaraInt32}
  User.SkeletalMesh {
   type=NiagaraDataInterfaceSkeletalMesh
   defaultValue {
    struct=NiagaraExt_VariableValue_DataInterface
    value {dataInterface=&self:NiagaraDataInterfaceSkeletalMesh_0}
   }
  }
  User.Transparency {type=NiagaraFloat, defaultValue=1, description=""}
 ]
}
